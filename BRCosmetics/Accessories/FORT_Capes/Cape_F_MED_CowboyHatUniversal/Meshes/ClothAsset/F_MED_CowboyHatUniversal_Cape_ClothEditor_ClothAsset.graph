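# Chaos Cloth asset for the F_MED_CowboyHatUniversal cape, built on the shared Fortnite Base Cape v3 Dataflow.
# Imported sim and render meshes get transferred skin weights, max distance, tethers, anim drive, collision and an accessory-mesh backstop.
# MaterialWeight selects among three presets for density, aerodynamics and bending.
# LOD0 takes that result directly, while LOD1 uses PBD stretch and bending, a density-map-remeshed sim mesh and proxy-deformed LOD1 render mesh.
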ChaosClothAsset /BRCosmetics/Accessories/FORT_Capes/Cape_F_MED_CowboyHatUniversal/Meshes/ClothAsset/F_MED_CowboyHatUniversal_Cape_ClothEditor_ClothAsset
name=Fortnite_Base_Cape_v3_Dataflow
graph=/BRCosmetics/Accessories/FORT_Capes/Common/Base_Cape_v3/Fortnite_Base_Cape_v3_Dataflow
connections [
 {
  fromNode=SimulationMaxDistanceConfig
  fromPinGuid=#1
  toNode=SimulationLongRangeAttachmentConfig
  toPinGuid=#2
 }
 {
  fromNode=WeightMap MaxDist
  fromPinGuid=#3
  toNode=SimulationMaxDistanceConfig
  toPinGuid=#4
 }
 {
  fromNode=SimulationDampingConfig_0
  fromPinGuid=#5
  toNode=Select
  toPin="Inputs[2]"
  toPinGuid=#6
 }
 {
  fromNode=ReRouteNode_v1_1
  fromPin=Value
  fromPinGuid=#7
  toNode=ReRouteNode_v1
  toPin=Value
  toPinGuid=#8
 }
 {
  fromNode=WeightMap Lift and Drag
  fromPin=OutputName.StringValue
  fromPinGuid=#9
  toNode=SimulationAerodynamicsConfig
  toPin=Drag.WeightMap
  toPinGuid=#10
 }
 {
  fromNode=WeightMap Lift and Drag
  fromPin=OutputName.StringValue
  fromPinGuid=#9
  toNode=SimulationAerodynamicsConfig_0
  toPin=Drag.WeightMap
  toPinGuid=#11
 }
 {
  fromNode=MaterialWeight
  fromPin=Value
  fromPinGuid=#12
  toNode=Select
  toPin=SelectedIndex
  toPinGuid=#13
 }
 {
  fromNode=SkeletalMesh
  fromPin=SkeletalMesh
  fromPinGuid=#14
  toNode=ReRouteNode_v1_1
  toPin=Value
  toPinGuid=#15
 }
 {
  fromNode=WeightMap Cloth Collision Thickness
  fromPin=OutputName.StringValue
  fromPinGuid=#16
  toNode=SimulationCollisionConfig
  toPin=ClothCollisionThicknessImported.WeightMap
  toPinGuid=#17
 }
 {
  fromNode=SimulationDampingConfig
  fromPinGuid=#18
  toNode=Select
  toPin="Inputs[1]"
  toPinGuid=#19
 }
 {
  fromNode=SimulationGravityConfig
  fromPinGuid=#20
  toNode=SimulationSolverConfig_01
  toPinGuid=#21
 }
 {
  fromNode=SimulationBendingConfig_01
  fromPinGuid=#22
  toNode=SimulationStretchConfig_01
  toPinGuid=#23
 }
 {
  fromNode=WeightMap Lift and Drag
  fromPin=OutputName.StringValue
  fromPinGuid=#9
  toNode=SimulationAerodynamicsConfig
  toPin=Lift.WeightMap
  toPinGuid=#24
 }
 {
  fromNode=Select_0
  fromPin=Result
  fromPinGuid=#25
  toNode=ClothAssetTerminal
  toPin="CollectionLods[0]"
  toPinGuid=#26
 }
 {
  fromNode=SimulationCollisionConfig
  fromPinGuid=#27
  toNode=SimulationGravityConfig
  toPinGuid=#28
 }
 {
  fromNode=SimulationAerodynamicsConfig_01
  fromPinGuid=#29
  toNode=SimulationVelocityScaleConfig_01
  toPinGuid=#30
 }
 {
  fromNode=SimulationVelocityScaleConfig_0
  fromPinGuid=#31
  toNode=SimulationMassConfig_0
  toPinGuid=#32
 }
 {
  fromNode=SimulationAnimDriveConfig
  fromPinGuid=#33
  toNode=SimulationBackstopConfig
  toPinGuid=#34
 }
 {
  fromNode=SkeletalMeshImport_0
  fromPinGuid=#35
  toNode=MergeClothCollections
  toPin="Collections[1]"
  toPinGuid=#36
 }
 {
  fromNode=WeightMap Lift and Drag
  fromPin=OutputName.StringValue
  fromPinGuid=#9
  toNode=SimulationAerodynamicsConfig_0
  toPin=Lift.WeightMap
  toPinGuid=#37
 }
 {
  fromNode=SkeletalMesh
  fromPin=SkeletalMesh
  fromPinGuid=#14
  toNode=SkeletalMeshImport
  toPin=SkeletalMesh
  toPinGuid=#38
 }
 {
  fromNode=WeightMap Lift and Drag
  fromPin=OutputName.StringValue
  fromPinGuid=#9
  toNode=SimulationAerodynamicsConfig_01
  toPin=Lift.WeightMap
  toPinGuid=#39
 }
 {
  fromNode=SkeletalMeshImport
  fromPinGuid=#40
  toNode=MergeClothCollections
  toPin="Collections[0]"
  toPinGuid=#41
 }
 {
  fromNode=SimulationBackstopConfig
  fromPinGuid=#42
  toNode=WeightMap Cloth Collision Thickness
  toPinGuid=#43
 }
 {
  fromNode=SkeletalMeshImport_Accessory Mesh
  fromPinGuid=#44
  toNode=SimAccessoryMeshNode
  toPin=SimAccessoryMeshCollection
  toPinGuid=#45
 }
 {
  fromNode=WeightMap Lift and Drag
  fromPin=OutputName.StringValue
  fromPinGuid=#9
  toNode=SimulationAerodynamicsConfig_01
  toPin=Drag.WeightMap
  toPinGuid=#46
 }
 {
  fromNode=WeightMap Cloth Collision Thickness
  fromPinGuid=#47
  toNode=SimulationCollisionConfig
  toPinGuid=#48
 }
 {
  fromNode=SimulationDampingConfig_01
  fromPinGuid=#49
  toNode=Select
  toPin="Inputs[0]"
  toPinGuid=#50
 }
 {
  fromNode=WeightMap Lift and Drag
  fromPinGuid=#51
  toNode=SimulationAerodynamicsConfig_01
  toPinGuid=#52
 }
 {
  fromNode=ReferenceBone
  fromPinGuid=#53
  toNode=WeightMap Lift and Drag
  toPinGuid=#54
 }
 {
  fromNode=SimulationSolverConfig_01
  fromPinGuid=#55
  toNode=ReferenceBone
  toPinGuid=#56
 }
 {
  fromNode=SimulationStretchConfig_0
  fromPinGuid=#57
  toNode=Select_0
  toPin="Inputs[2]"
  toPinGuid=#58
 }
 {
  fromNode=TransferSkinWeights_Cape_02
  fromPinGuid=#59
  toNode=WeightMap MaxDist
  toPinGuid=#60
 }
 {
  fromNode=SimulationMassConfig_01
  fromPinGuid=#61
  toNode=SimulationDampingConfig_01
  toPinGuid=#62
 }
 {
  fromNode=WeightMap MaxDist
  fromPin=OutputName.StringValue
  fromPinGuid=#63
  toNode=SimulationMaxDistanceConfig
  toPin=MaxDistance.WeightMap
  toPinGuid=#64
 }
 {
  fromNode=MergeClothCollections
  fromPinGuid=#65
  toNode=SimAccessoryMeshNode
  toPinGuid=#66
 }
 {
  fromNode=RenderSkelMesh
  fromPin=Value
  fromPinGuid=#67
  toNode=SkeletalMeshImport_0
  toPin=SkeletalMesh
  toPinGuid=#68
 }
 {
  fromNode=SimAccessoryMeshNode
  fromPinGuid=#69
  toNode=TransferSkinWeights_Cape_02
  toPinGuid=#70
 }
 {
  fromNode=ReRouteNode_v1
  fromPin=Value
  fromPinGuid=#71
  toNode=TransferSkinWeights_Cape_02
  toPin=SkeletalMesh
  toPinGuid=#72
 }
 {
  fromNode=SimulationVelocityScaleConfig_01
  fromPinGuid=#73
  toNode=SimulationMassConfig_01
  toPinGuid=#74
 }
 {
  fromNode=SimulationBendingConfig
  fromPinGuid=#75
  toNode=SimulationStretchConfig
  toPinGuid=#76
 }
 {
  fromNode=SimulationMassConfig
  fromPinGuid=#77
  toNode=SimulationDampingConfig
  toPinGuid=#78
 }
 {
  fromNode=SimulationAerodynamicsConfig
  fromPinGuid=#79
  toNode=SimulationVelocityScaleConfig
  toPinGuid=#80
 }
 {
  fromNode=Select
  fromPin=Result
  fromPinGuid=#81
  toNode=SimulationBendingConfig_01
  toPinGuid=#82
 }
 {
  fromNode=SimulationVelocityScaleConfig
  fromPinGuid=#83
  toNode=SimulationMassConfig
  toPinGuid=#84
 }
 {
  fromNode=SimulationBendingConfig_0
  fromPinGuid=#85
  toNode=SimulationStretchConfig_0
  toPinGuid=#86
 }
 {
  fromNode=SimulationMassConfig_0
  fromPinGuid=#87
  toNode=SimulationDampingConfig_0
  toPinGuid=#88
 }
 {
  fromNode=SimulationAerodynamicsConfig_0
  fromPinGuid=#89
  toNode=SimulationVelocityScaleConfig_0
  toPinGuid=#90
 }
 {
  fromNode=Select
  fromPin=Result
  fromPinGuid=#81
  toNode=SimulationBendingConfig
  toPinGuid=#91
 }
 {
  fromNode=WeightMap Lift and Drag
  fromPinGuid=#51
  toNode=SimulationAerodynamicsConfig
  toPinGuid=#92
 }
 {
  fromNode=WeightMap Lift and Drag
  fromPinGuid=#51
  toNode=SimulationAerodynamicsConfig_0
  toPinGuid=#93
 }
 {
  fromNode=ProxyDeformer
  fromPinGuid=#94
  toNode=ClothAssetTerminal
  toPin="CollectionLods[1]"
  toPinGuid=#95
 }
 {
  fromNode=DeleteElement
  fromPinGuid=#96
  toNode=WeightMap Remesh Density
  toPinGuid=#97
 }
 {fromNode=WeightMap Remesh Density, fromPinGuid=#98, toNode=Remesh, toPinGuid=#99}
 {
  fromNode=SimulationLongRangeAttachmentConfig
  fromPinGuid=#100
  toNode=WeightMap AnimDriveStiffness
  toPinGuid=#101
 }
 {
  fromNode=WeightMap Remesh Density
  fromPin=OutputName.StringValue
  fromPinGuid=#102
  toNode=Remesh
  toPin=DensityMapSim.WeightMap
  toPinGuid=#103
 }
 {
  fromNode=Remesh
  fromPinGuid=#104
  toNode=MergeClothCollections_1
  toPin="Collections[0]"
  toPinGuid=#105
 }
 {
  fromNode=SkeletalMeshImport_Render Mesh
  fromPinGuid=#106
  toNode=MergeClothCollections_1
  toPin="Collections[1]"
  toPinGuid=#107
 }
 {
  fromNode=ProceduralSelection_1
  fromPinGuid=#108
  toNode=ProxyDeformer
  toPinGuid=#109
 }
 {
  fromNode=ProceduralSelection
  fromPin=OutputName
  fromPinGuid=#110
  toNode=ProxyDeformer
  toPin="SelectionFilterSets[0].RenderSelection.StringValue"
  toPinGuid=#111
 }
 {
  fromNode=ProceduralSelection_1
  fromPin=OutputName
  fromPinGuid=#112
  toNode=ProxyDeformer
  toPin="SelectionFilterSets[0].SimSelection.StringValue"
  toPinGuid=#113
 }
 {
  fromNode=MergeClothCollections_1
  fromPinGuid=#114
  toNode=ProceduralSelection
  toPinGuid=#115
 }
 {
  fromNode=ProceduralSelection
  fromPinGuid=#116
  toNode=ProceduralSelection_1
  toPinGuid=#117
 }
 {
  fromNode=RenderSkelMesh_0
  fromPin=Value
  fromPinGuid=#118
  toNode=SkeletalMeshImport_Render Mesh
  toPin=SkeletalMesh
  toPinGuid=#119
 }
 {
  fromNode=WeightMap AnimDriveStiffness
  fromPinGuid=#120
  toNode=SimulationAnimDriveConfig
  toPinGuid=#121
 }
 {
  fromNode=WeightMap AnimDriveStiffness
  fromPin=OutputName.StringValue
  fromPinGuid=#122
  toNode=SimulationAnimDriveConfig
  toPin=AnimDriveStiffness.WeightMap
  toPinGuid=#123
 }
 {
  fromNode=Select
  fromPin=Result
  fromPinGuid=#81
  toNode=SimulationBendingConfig_0
  toPinGuid=#124
 }
 {
  fromNode=SimulationStretchConfig_01
  fromPinGuid=#125
  toNode=Select_0
  toPin="Inputs[0]"
  toPinGuid=#126
 }
 {
  fromNode=SimulationStretchConfig
  fromPinGuid=#127
  toNode=Select_0
  toPin="Inputs[1]"
  toPinGuid=#128
 }
 {
  fromNode=MaterialWeight_2
  fromPin=Value
  fromPinGuid=#129
  toNode=Select_0
  toPin=SelectedIndex
  toPinGuid=#130
 }
 {
  fromNode=SimulationBendingConfig_1
  fromPinGuid=#131
  toNode=SimulationStretchConfig_1
  toPinGuid=#132
 }
 {
  fromNode=SimulationBendingConfig_2
  fromPinGuid=#133
  toNode=SimulationStretchConfig_2
  toPinGuid=#134
 }
 {
  fromNode=SimulationBendingConfig_3
  fromPinGuid=#135
  toNode=SimulationStretchConfig_3
  toPinGuid=#136
 }
 {
  fromNode=Select
  fromPin=Result
  fromPinGuid=#81
  toNode=SimulationBendingConfig_1
  toPinGuid=#137
 }
 {
  fromNode=Select
  fromPin=Result
  fromPinGuid=#81
  toNode=SimulationBendingConfig_2
  toPinGuid=#138
 }
 {
  fromNode=Select
  fromPin=Result
  fromPinGuid=#81
  toNode=SimulationBendingConfig_3
  toPinGuid=#139
 }
 {
  fromNode=MaterialWeight_3
  fromPin=Value
  fromPinGuid=#140
  toNode=Select_1
  toPin=SelectedIndex
  toPinGuid=#141
 }
 {
  fromNode=SimulationStretchConfig_1
  fromPinGuid=#142
  toNode=Select_1
  toPin="Inputs[0]"
  toPinGuid=#143
 }
 {
  fromNode=SimulationStretchConfig_2
  fromPinGuid=#144
  toNode=Select_1
  toPin="Inputs[1]"
  toPinGuid=#145
 }
 {
  fromNode=SimulationStretchConfig_3
  fromPinGuid=#146
  toNode=Select_1
  toPin="Inputs[2]"
  toPinGuid=#147
 }
 {
  fromNode=Select_1
  fromPin=Result
  fromPinGuid=#148
  toNode=DeleteElement
  toPinGuid=#149
 }
]
nodes [
 SimulationStretchConfig_01 {
  type=FChaosClothAssetSimulationStretchConfigNode
  guid=#150
  properties {
   StretchAnisoDamping=(bIsAnimatable=True,Low=2,High=2,WeightMap="StretchAnisoDamping")
   StretchStiffnessBias=(bIsAnimatable=True,Low=100,High=100,WeightMap="StretchStiffnessBias")
   StretchStiffnessWarp=(bIsAnimatable=True,Low=100000,High=10000,WeightMap="StretchStiffnessWarp")
   StretchStiffnessWeft=(bIsAnimatable=True,Low=100000,High=10000,WeightMap="StretchStiffnessWeft")
  }
 }
 SimulationBendingConfig_01 {
  type=FChaosClothAssetSimulationBendingConfigNode
  guid=#151
  inputPins [
   Collection {type=FManagedArrayCollection}
   FlatnessRatio.WeightMap {type=FString}
   RestAngle.WeightMap {type=FString}
   BendingStiffness.WeightMap {type=FString}
   BendingStiffnessWarp.WeightMap {type=FString}
   BendingStiffnessWeft.WeightMap {type=FString}
   BendingStiffnessBias.WeightMap {type=FString}
   BendingDamping.WeightMap {type=FString}
   BendingAnisoDamping.WeightMap {type=FString}
   AnisoBucklingRatio.WeightMap {type=FString}
   BucklingStiffness.WeightMap {type=FString}
   BucklingStiffnessWarp.WeightMap {type=FString}
   BucklingStiffnessWeft.WeightMap {type=FString}
   BucklingStiffnessBias.WeightMap {type=FString}
   BucklingRatioWeighted.WeightMap {type=FString}
  ]
  properties {
   AnisoBucklingRatio=(bIsAnimatable=True,Low=0.5,High=0.5,WeightMap="BucklingRatio")
   BendingAnisoDamping=(bIsAnimatable=True,Low=10,High=1,WeightMap="BendingAnisoDamping")
   BendingStiffnessBias=(bIsAnimatable=True,Low=1000,High=100,WeightMap="BendingStiffnessBias")
   BendingStiffnessWarp=(bIsAnimatable=True,Low=1000,High=100,WeightMap="BendingStiffnessWarp")
   BendingStiffnessWeft=(bIsAnimatable=True,Low=1000,High=100,WeightMap="BendingStiffnessWeft")
   BucklingStiffnessBias=(bIsAnimatable=True,Low=50,High=50,WeightMap="BucklingStiffnessBias")
   BucklingStiffnessWarp=(bIsAnimatable=True,Low=50,High=50,WeightMap="BucklingStiffnessWarp")
   BucklingStiffnessWeft=(bIsAnimatable=True,Low=50,High=50,WeightMap="BucklingStiffnessWeft")
  }
 }
 SimulationSolverConfig_01 {
  type=FChaosClothAssetSimulationSolverConfigNode
  guid=#152
  properties {MaxNumIterations=2, NumIterations=1, NumSubstepsImported=(ImportedValue=2)}
 }
 SimulationMassConfig_01 {
  type=FChaosClothAssetSimulationMassConfigNode
  guid=#153
  properties {DensityWeighted=(Low=1,High=1,WeightMap="Density")}
 }
 SimulationAerodynamicsConfig_01 {
  type=FChaosClothAssetSimulationAerodynamicsConfigNode
  guid=#154
  properties {
   Drag=(bIsAnimatable=True,Low=0.2,High=0.6,WeightMap="Drag")
   Lift=(bIsAnimatable=True,Low=0.2,High=0.3,WeightMap="Lift")
   TurbulenceRatio=1
   WindVelocity=(X=0,Y=0,Z=0)
   WindVelocitySpace=ComponentSpace
  }
 }
 SimulationDampingConfig_01 {
  type=FChaosClothAssetSimulationDampingConfigNode
  guid=#155
  properties {
   DampingCoefficientWeighted=(Low=0.01,High=0.01,WeightMap="DampingCoefficient")
   LocalDampingAngularCoefficient=0.05
   LocalDampingLinearCoefficient=0.05
   LocalDampingSpace=ReferenceBoneSpace
  }
 }
 SimulationGravityConfig {
  type=FChaosClothAssetSimulationGravityConfigNode
  guid=#156
  properties {GravityOverrideImported=(ImportedValue=(X=0,Y=0,Z=-980.665))}
 }
 ClothAssetTerminal {
  type=FChaosClothAssetTerminalNode_v2
  guid=#157
  inputPins [
   "CollectionLods[0]" {type=FManagedArrayCollection}
   "CollectionLods[1]" {type=FManagedArrayCollection}
  ]
 }
 SimulationCollisionConfig {
  type=FChaosClothAssetSimulationCollisionConfigNode
  guid=#158
  properties {
   ClothCollisionThicknessImported=(bIsAnimatable=True,Low=0.1,High=3,WeightMap="ClothCollisionThickness")
   CollisionThicknessImported=(ImportedValue=1)
   FrictionCoefficientWeighted=(bIsAnimatable=True,Low=0.5,High=0.5,WeightMap="FrictionCoefficient")
   InnerCollisionThickness=(Low=1,High=1,WeightMap="InnerCollisionThickness")
   bUseCCD=False
  }
 }
 SimulationBackstopConfig {
  type=FChaosClothAssetSimulationBackstopConfigNode
  guid=#159
  inputPins [
   Collection {type=FManagedArrayCollection}
   BackstopDistance.WeightMap {type=FString}
   BackstopRadius.WeightMap {type=FString}
   BackstopMeshName.StringValue {type=FString}
  ]
  outputPins [Collection {type=FManagedArrayCollection}]
  properties {
   BackstopDistance=(bIsAnimatable=True,High=-10,WeightMap="BackstopDistance")
   BackstopMeshName=(StringValue="AccessoryMesh")
   BackstopRadius=(bIsAnimatable=True,Low=100,High=100,WeightMap="BackstopRadius")
   bWarnDuplicateProperty=True
  }
 }
 SkeletalMeshImport {
  type=FChaosClothAssetSkeletalMeshImportNode_v2
  guid=#160
  properties {SkeletalMesh=None, bImportRenderMesh=False, bImportSimMesh=True, bImportSingleSection=True}
 }
 SkeletalMesh {
  type=FGetSkeletalMeshDataflowNode
  guid=#161
  inputPins []
  outputPins [SkeletalMesh {type="TObjectPtr<USkeletalMesh>"}]
  properties {
   PropertyName=SkeletalMesh
   SkeletalMesh=&SkeletalMesh'/BRCosmetics/Accessories/FORT_Capes/Common/Base_Cape_v3/M_MED_Base_Cape_v3_SIM'
  }
 }
 TransferSkinWeights_Cape_02 {
  type=FChaosClothAssetTransferSkinWeightsNode
  guid=#162
  inputPins [
   Collection {type=FManagedArrayCollection}
   SkeletalMesh {type="TObjectPtr<USkeletalMesh>"}
   LodIndex {type=int32}
   SimCollection {type=FManagedArrayCollection}
   InpaintMask.WeightMap {type=FString}
  ]
  outputPins [Collection {type=FManagedArrayCollection}]
  properties {
   InpaintMask=(WeightMap="InpaintMask")
   LayeredMeshSupport=True
   LodIndex=0
   MaxNumInfluences=Eight
   NormalThreshold=30
   NumSmoothingIterations=10
   RadiusPercentage=0.05
   RenderMeshSourceType=SkeletalMesh
   SimMeshSourceTypeHint=Skeletal Mesh
   SkeletalMesh=&SkeletalMesh'&M_MED_Base_Cape_v3_SIM'
   SmoothingStrength=0.1
   TargetMeshType=Render
   TransferMethod=InpaintWeights
   TransferMethodHint=Closest Point On Surface
   Transform=(Rotation=(X=0,Y=0,Z=0,W=1),Translation=(X=0,Y=0,Z=0),Scale3D=(X=1,Y=1,Z=1))
  }
 }
 MergeClothCollections {type=FChaosClothAssetMergeClothCollectionsNode_v2, guid=#163}
 SimulationMaxDistanceConfig {
  type=FChaosClothAssetSimulationMaxDistanceConfigNode
  guid=#164
  properties {MaxDistance=(bIsAnimatable=True,High=300,WeightMap="MaxDistance")}
 }
 WeightMap MaxDist {
  type=FChaosClothAssetWeightMapNode
  guid=#165
  properties {InputName=(StringValue="MaxDistance_Ext"), MapOverrideType=ReplaceAll, OutputName=(StringValue="MaxDistance"), TransferType=Use3DSimMesh}
 }
 SimulationLongRangeAttachmentConfig {
  type=FChaosClothAssetSimulationLongRangeAttachmentConfigNode_v2
  guid=#166
  inputPins [
   Collection {type=FManagedArrayCollection}
   FixedEndSet.StringValue {type=FString}
   TetherStiffness.WeightMap {type=FString}
   TetherScale.WeightMap {type=FString}
   "CustomTetherData[0].CustomFixedEndSet.StringValue" {type=FString}
   "CustomTetherData[0].CustomDynamicEndSet.StringValue" {type=FString}
  ]
  outputPins [Collection {type=FManagedArrayCollection}]
  properties {
   CustomTetherData=(())
   FixedEndSet=(StringValue="KinematicVertices3D")
   TetherScale=(bIsAnimatable=True,Low=1,High=1,WeightMap="TetherScale")
   TetherStiffness=(bIsAnimatable=True,Low=1,High=1,WeightMap="TetherStiffness")
   bEnableCustomTetherGeneration=False
   bUseGeodesicTethers=True
   bWarnDuplicateProperty=True
  }
 }
 SimulationVelocityScaleConfig_01 {
  type=FChaosClothAssetSimulationVelocityScaleConfigNode
  guid=#167
  inputPins [Collection {type=FManagedArrayCollection}]
  outputPins [Collection {type=FManagedArrayCollection}]
  properties {
   AngularVelocityScale=0.5
   FictitiousAngularScale=0.6
   FictitiousForcesModel=FullScale
   LinearVelocityScale=(X=0.5,Y=0.5,Z=0.5)
   MaxAngularAcceleration=9
   MaxAngularVelocity=2
   MaxLinearAcceleration=(X=1800,Y=1800,Z=1800)
   MaxLinearVelocity=(X=800,Y=800,Z=800)
   MaxVelocityScale=2
   VelocityScaleSpace=ReferenceBoneSpace
   bEnableAngularAccelerationClamping=True
   bEnableAngularVelocityClamping=True
   bEnableLinearAccelerationClamping=True
   bEnableLinearVelocityClamping=True
   bWarnDuplicateProperty=True
  }
 }
 Select {
  type=FDataflowSelectNode
  guid=#168
  inputPins [
   SelectedIndex {type=int32}
   "Inputs[0]" {type=FManagedArrayCollection}
   "Inputs[1]" {type=FManagedArrayCollection}
   "Inputs[2]" {type=FManagedArrayCollection}
  ]
  outputPins [Result {type=FManagedArrayCollection}]
  properties {SelectedIndex=0}
 }
 WeightMap Lift and Drag {
  type=FChaosClothAssetWeightMapNode
  guid=#169
  properties {MapOverrideType=Add, OutputName=(StringValue="LiftAndDragMap"), TransferType=Use3DSimMesh}
 }
 ReferenceBone {
  type=FChaosClothAssetReferenceBoneNode
  guid=#170
  inputPins [Collection {type=FManagedArrayCollection}]
  outputPins [Collection {type=FManagedArrayCollection}]
  properties {CalculateDefaultReferenceBone=(), ReferenceBone=(Name="root")}
 }
 SkeletalMeshImport_0 {
  type=FChaosClothAssetSkeletalMeshImportNode_v2
  guid=#171
  properties {SkeletalMesh=None, bImportRenderMesh=True, bImportSimMesh=False, bImportSingleSection=True}
 }
 SimAccessoryMeshNode {
  type=FChaosClothAssetSimAccessoryMeshNode
  guid=#172
  inputPins [
   Collection {type=FManagedArrayCollection}
   SimAccessoryMeshCollection {type=FManagedArrayCollection}
  ]
  outputPins [
   Collection {type=FManagedArrayCollection}
   AccessoryMeshName {type=FString}
  ]
  properties {AccessoryMeshName=AccessoryMesh, bUseSimImportVertexID=True}
 }
 SkeletalMeshImport_Accessory Mesh {
  type=FChaosClothAssetSkeletalMeshImportNode_v2
  guid=#173
  properties {
   SkeletalMesh=&SkeletalMesh'/BRCosmetics/Accessories/FORT_Capes/Common/Base_Cape_v3/M_MED_Base_Cape_v3_BACKSTOP'
   UVScale=(X=120,Y=120)
   bImportRenderMesh=False
   bImportSimMesh=True
   bImportSingleSection=False
  }
 }
 MaterialWeight {
  type=FGetDataflowVariableNode
  guid=#174
  outputPins [Value {type=int32}]
  properties {
   Category=Condition
   VariableName=MaterialWeight
   VariablePropertyBag=(Value=/Engine/Transient.PropertyBag_33cf849a4a835154(MaterialWeight=1))
  }
 }
 RenderSkelMesh {
  type=FGetDataflowVariableNode
  guid=#175
  outputPins [Value {type="TObjectPtr<USkeletalMesh>"}]
  properties {
   VariableName=RenderSkelMesh
   VariablePropertyBag=(Value=/Engine/Transient.PropertyBag_22397d724b0fa344(RenderSkelMesh="&SkeletalMesh'/BRCosmetics/Accessories/FORT_Capes/Common/Base_Cape_v3/M_MED_Base_Cape_v3'"))
  }
 }
 WeightMap Cloth Collision Thickness {type=FChaosClothAssetWeightMapNode, guid=#176, properties {OutputName=(StringValue="ClothCollisionThickness")}}
 SimulationStretchConfig {
  type=FChaosClothAssetSimulationStretchConfigNode
  guid=#177
  properties {
   StretchAnisoDamping=(bIsAnimatable=True,Low=2,High=2,WeightMap="StretchAnisoDamping")
   StretchStiffnessBias=(bIsAnimatable=True,Low=100,High=100,WeightMap="StretchStiffnessBias")
   StretchStiffnessWarp=(bIsAnimatable=True,Low=100000,High=10000,WeightMap="StretchStiffnessWarp")
   StretchStiffnessWeft=(bIsAnimatable=True,Low=100000,High=10000,WeightMap="StretchStiffnessWeft")
  }
 }
 SimulationBendingConfig {
  type=FChaosClothAssetSimulationBendingConfigNode
  guid=#178
  inputPins [
   Collection {type=FManagedArrayCollection}
   FlatnessRatio.WeightMap {type=FString}
   RestAngle.WeightMap {type=FString}
   BendingStiffness.WeightMap {type=FString}
   BendingStiffnessWarp.WeightMap {type=FString}
   BendingStiffnessWeft.WeightMap {type=FString}
   BendingStiffnessBias.WeightMap {type=FString}
   BendingDamping.WeightMap {type=FString}
   BendingAnisoDamping.WeightMap {type=FString}
   AnisoBucklingRatio.WeightMap {type=FString}
   BucklingStiffness.WeightMap {type=FString}
   BucklingStiffnessWarp.WeightMap {type=FString}
   BucklingStiffnessWeft.WeightMap {type=FString}
   BucklingStiffnessBias.WeightMap {type=FString}
   BucklingRatioWeighted.WeightMap {type=FString}
  ]
  properties {
   AnisoBucklingRatio=(bIsAnimatable=True,Low=0.5,High=0.5,WeightMap="BucklingRatio")
   BendingAnisoDamping=(bIsAnimatable=True,Low=3,High=1,WeightMap="BendingAnisoDamping")
   BendingStiffnessBias=(bIsAnimatable=True,Low=300,High=100,WeightMap="BendingStiffnessBias")
   BendingStiffnessWarp=(bIsAnimatable=True,Low=300,High=100,WeightMap="BendingStiffnessWarp")
   BendingStiffnessWeft=(bIsAnimatable=True,Low=300,High=100,WeightMap="BendingStiffnessWeft")
   BucklingStiffnessBias=(bIsAnimatable=True,Low=50,High=50,WeightMap="BucklingStiffnessBias")
   BucklingStiffnessWarp=(bIsAnimatable=True,Low=50,High=50,WeightMap="BucklingStiffnessWarp")
   BucklingStiffnessWeft=(bIsAnimatable=True,Low=50,High=50,WeightMap="BucklingStiffnessWeft")
  }
 }
 SimulationMassConfig {
  type=FChaosClothAssetSimulationMassConfigNode
  guid=#179
  properties {DensityWeighted=(Low=0.7,High=0.7,WeightMap="Density")}
 }
 SimulationAerodynamicsConfig {
  type=FChaosClothAssetSimulationAerodynamicsConfigNode
  guid=#180
  properties {
   Drag=(bIsAnimatable=True,Low=0.3,High=0.8,WeightMap="Drag")
   Lift=(bIsAnimatable=True,Low=0.2,High=0.4,WeightMap="Lift")
   TurbulenceRatio=1
   WindVelocity=(X=0,Y=0,Z=0)
   WindVelocitySpace=ComponentSpace
  }
 }
 SimulationDampingConfig {
  type=FChaosClothAssetSimulationDampingConfigNode
  guid=#181
  properties {
   DampingCoefficientWeighted=(Low=0.01,High=0.01,WeightMap="DampingCoefficient")
   LocalDampingAngularCoefficient=0.05
   LocalDampingLinearCoefficient=0.05
   LocalDampingSpace=ReferenceBoneSpace
  }
 }
 SimulationVelocityScaleConfig {
  type=FChaosClothAssetSimulationVelocityScaleConfigNode
  guid=#182
  inputPins [Collection {type=FManagedArrayCollection}]
  outputPins [Collection {type=FManagedArrayCollection}]
  properties {
   AngularVelocityScale=0.5
   FictitiousAngularScale=0.6
   FictitiousForcesModel=FullScale
   LinearVelocityScale=(X=0.5,Y=0.5,Z=0.5)
   MaxAngularAcceleration=9
   MaxAngularVelocity=2
   MaxLinearAcceleration=(X=1800,Y=1800,Z=1800)
   MaxLinearVelocity=(X=800,Y=800,Z=800)
   MaxVelocityScale=2
   VelocityScaleSpace=ReferenceBoneSpace
   bEnableAngularAccelerationClamping=True
   bEnableAngularVelocityClamping=True
   bEnableLinearAccelerationClamping=True
   bEnableLinearVelocityClamping=True
   bWarnDuplicateProperty=True
  }
 }
 SimulationStretchConfig_0 {
  type=FChaosClothAssetSimulationStretchConfigNode
  guid=#183
  properties {
   StretchAnisoDamping=(bIsAnimatable=True,Low=2,High=2,WeightMap="StretchAnisoDamping")
   StretchStiffnessBias=(bIsAnimatable=True,Low=100,High=100,WeightMap="StretchStiffnessBias")
   StretchStiffnessWarp=(bIsAnimatable=True,Low=100000,High=10000,WeightMap="StretchStiffnessWarp")
   StretchStiffnessWeft=(bIsAnimatable=True,Low=100000,High=10000,WeightMap="StretchStiffnessWeft")
  }
 }
 SimulationBendingConfig_0 {
  type=FChaosClothAssetSimulationBendingConfigNode
  guid=#184
  inputPins [
   Collection {type=FManagedArrayCollection}
   FlatnessRatio.WeightMap {type=FString}
   RestAngle.WeightMap {type=FString}
   BendingStiffness.WeightMap {type=FString}
   BendingStiffnessWarp.WeightMap {type=FString}
   BendingStiffnessWeft.WeightMap {type=FString}
   BendingStiffnessBias.WeightMap {type=FString}
   BendingDamping.WeightMap {type=FString}
   BendingAnisoDamping.WeightMap {type=FString}
   AnisoBucklingRatio.WeightMap {type=FString}
   BucklingStiffness.WeightMap {type=FString}
   BucklingStiffnessWarp.WeightMap {type=FString}
   BucklingStiffnessWeft.WeightMap {type=FString}
   BucklingStiffnessBias.WeightMap {type=FString}
   BucklingRatioWeighted.WeightMap {type=FString}
  ]
  properties {
   AnisoBucklingRatio=(bIsAnimatable=True,Low=0.5,High=0.5,WeightMap="BucklingRatio")
   BendingAnisoDamping=(bIsAnimatable=True,Low=1,High=1,WeightMap="BendingAnisoDamping")
   BendingStiffnessBias=(bIsAnimatable=True,Low=100,High=100,WeightMap="BendingStiffnessBias")
   BendingStiffnessWarp=(bIsAnimatable=True,Low=100,High=100,WeightMap="BendingStiffnessWarp")
   BendingStiffnessWeft=(bIsAnimatable=True,Low=100,High=100,WeightMap="BendingStiffnessWeft")
   BucklingStiffnessBias=(bIsAnimatable=True,Low=50,High=50,WeightMap="BucklingStiffnessBias")
   BucklingStiffnessWarp=(bIsAnimatable=True,Low=50,High=50,WeightMap="BucklingStiffnessWarp")
   BucklingStiffnessWeft=(bIsAnimatable=True,Low=50,High=50,WeightMap="BucklingStiffnessWeft")
  }
 }
 SimulationMassConfig_0 {
  type=FChaosClothAssetSimulationMassConfigNode
  guid=#185
  properties {DensityWeighted=(Low=0.4,High=0.4,WeightMap="Density")}
 }
 SimulationAerodynamicsConfig_0 {
  type=FChaosClothAssetSimulationAerodynamicsConfigNode
  guid=#186
  properties {
   Drag=(bIsAnimatable=True,Low=0.4,High=1.3,WeightMap="Drag")
   Lift=(bIsAnimatable=True,Low=0.4,High=0.8,WeightMap="Lift")
   TurbulenceRatio=1
   WindVelocity=(X=0,Y=0,Z=0)
   WindVelocitySpace=ComponentSpace
  }
 }
 SimulationDampingConfig_0 {
  type=FChaosClothAssetSimulationDampingConfigNode
  guid=#187
  properties {
   DampingCoefficientWeighted=(Low=0.01,High=0.01,WeightMap="DampingCoefficient")
   LocalDampingAngularCoefficient=0.05
   LocalDampingLinearCoefficient=0.05
   LocalDampingSpace=ReferenceBoneSpace
  }
 }
 SimulationVelocityScaleConfig_0 {
  type=FChaosClothAssetSimulationVelocityScaleConfigNode
  guid=#188
  inputPins [Collection {type=FManagedArrayCollection}]
  outputPins [Collection {type=FManagedArrayCollection}]
  properties {
   AngularVelocityScale=0.5
   FictitiousAngularScale=0.6
   FictitiousForcesModel=FullScale
   LinearVelocityScale=(X=0.5,Y=0.5,Z=0.5)
   MaxAngularAcceleration=9
   MaxAngularVelocity=2
   MaxLinearAcceleration=(X=1800,Y=1800,Z=1800)
   MaxLinearVelocity=(X=800,Y=800,Z=800)
   MaxVelocityScale=2
   VelocityScaleSpace=ReferenceBoneSpace
   bEnableAngularAccelerationClamping=True
   bEnableAngularVelocityClamping=True
   bEnableLinearAccelerationClamping=True
   bEnableLinearVelocityClamping=True
   bWarnDuplicateProperty=True
  }
 }
 ReRouteNode_v1 {
  type=FDataflowReRouteNode
  guid=#189
  inputPins [Value {type="TObjectPtr<USkeletalMesh>"}]
  outputPins [Value {type="TObjectPtr<USkeletalMesh>"}]
 }
 ReRouteNode_v1_1 {
  type=FDataflowReRouteNode
  guid=#190
  inputPins [Value {type="TObjectPtr<USkeletalMesh>"}]
  outputPins [Value {type="TObjectPtr<USkeletalMesh>"}]
 }
 Remesh {
  type=FChaosClothAssetRemeshNode_v2
  guid=#191
  inputPins [
   Collection {type=FManagedArrayCollection}
   TriangleSelectionSim.StringValue {type=FString}
   DensityMapSim.WeightMap {type=FString}
   TriangleSelectionRender.StringValue {type=FString}
   DensityMapRender.WeightMap {type=FString}
  ]
  outputPins [Collection {type=FManagedArrayCollection}]
  properties {
   DensityMapRender=(Low=100,High=200,WeightMap="DensityMapRender")
   DensityMapSim=(Low=30,High=75,WeightMap="DensityMapSim")
   DensityMapThresholdRender=1
   DensityMapThresholdSim=1
   IterationsRender=10
   IterationsSim=50
   RemeshMethodRender=Remesh
   RenderSeamRemeshIterations=1
   SelectionSourceRender=DensityMapInput
   SelectionSourceSim=DensityMapInput
   SmoothingRender=0.25
   SmoothingSim=0.5
   TargetPercentRender=100
   TriangleSelectionRender=()
   TriangleSelectionSim=()
   bRemeshRender=False
   bRemeshRenderSeams=False
   bRemeshSim=True
  }
 }
 DeleteElement {
  type=FChaosClothAssetDeleteElementNode
  guid=#192
  inputPins [
   Collection {type=FManagedArrayCollection}
   SelectionName.StringValue {type=FString}
  ]
  outputPins [Collection {type=FManagedArrayCollection}]
  properties {Elements="", Group=(Name="RenderVertices"), SelectionName=(), bDeleteRenderMesh=True, bDeleteSimMesh=False}
 }
 SkeletalMeshImport_Render Mesh {
  type=FChaosClothAssetSkeletalMeshImportNode_v2
  guid=#193
  properties {LODIndex=1, SectionIndex=2, SkeletalMesh=None, bImportRenderMesh=True, bImportSimMesh=False, bImportSingleSection=False}
 }
 MergeClothCollections_1 {type=FChaosClothAssetMergeClothCollectionsNode_v2, guid=#194}
 ProxyDeformer {
  type=FChaosClothAssetProxyDeformerNode_v3
  guid=#195
  properties {InfluenceRadius=5, SelectionFilterSets=(()), bPreserveRenderTangents=True}
 }
 ProceduralSelection {
  type=FChaosClothAssetProceduralSelectionNode
  guid=#196
  inputPins [
   Collection {type=FManagedArrayCollection}
   ConversionInputName.StringValue {type=FString}
  ]
  outputPins [
   Collection {type=FManagedArrayCollection}
   OutputName {type=FString}
  ]
  properties {ConversionInputName=(), Group=(Name="RenderVertices"), OutputName=RenderSelectionAll, SelectionType=SelectAll}
 }
 ProceduralSelection_1 {
  type=FChaosClothAssetProceduralSelectionNode
  guid=#197
  inputPins [
   Collection {type=FManagedArrayCollection}
   ConversionInputName.StringValue {type=FString}
  ]
  outputPins [
   Collection {type=FManagedArrayCollection}
   OutputName {type=FString}
  ]
  properties {ConversionInputName=(), Group=(Name="SimVertices3D"), OutputName=SimSelectionAll, SelectionType=SelectAll}
 }
 WeightMap Remesh Density {type=FChaosClothAssetWeightMapNode, guid=#198, properties {OutputName=(StringValue="LOD1_RemeshDensity")}}
 RenderSkelMesh_0 {
  type=FGetDataflowVariableNode
  guid=#199
  outputPins [Value {type="TObjectPtr<USkeletalMesh>"}]
  properties {
   VariableName=RenderSkelMesh
   VariablePropertyBag=(Value=&PropertyBag_22397d724b0fa344(RenderSkelMesh="&SkeletalMesh'&M_MED_Base_Cape_v3'"))
  }
 }
 SimulationAnimDriveConfig {
  type=FChaosClothAssetSimulationAnimDriveConfigNode
  guid=#200
  inputPins [
   Collection {type=FManagedArrayCollection}
   AnimDriveStiffness.WeightMap {type=FString}
   AnimDriveDamping.WeightMap {type=FString}
  ]
  outputPins [Collection {type=FManagedArrayCollection}]
  properties {
   AnimDriveDamping=(bIsAnimatable=True,High=1,WeightMap="AnimDriveDamping")
   AnimDriveStiffness=(bIsAnimatable=True,High=1,WeightMap="AnimDriveStiffness")
   bWarnDuplicateProperty=True
  }
 }
 WeightMap AnimDriveStiffness {type=FChaosClothAssetWeightMapNode, guid=#201, properties {OutputName=(StringValue="AnimDriveStiffness")}}
 Select_0 {
  type=FDataflowSelectNode
  guid=#202
  inputPins [
   SelectedIndex {type=int32}
   "Inputs[0]" {type=FManagedArrayCollection}
   "Inputs[1]" {type=FManagedArrayCollection}
   "Inputs[2]" {type=FManagedArrayCollection}
  ]
  outputPins [Result {type=FManagedArrayCollection}]
  properties {SelectedIndex=0}
 }
 MaterialWeight_2 {
  type=FGetDataflowVariableNode
  guid=#203
  outputPins [Value {type=int32}]
  properties {
   Category=Condition
   VariableName=MaterialWeight
   VariablePropertyBag=(Value=&PropertyBag_33cf849a4a835154(MaterialWeight=1))
  }
 }
 SimulationStretchConfig_1 {
  type=FChaosClothAssetSimulationStretchConfigNode
  guid=#204
  properties {
   SolverType=PBD
   StretchAnisoDamping=(bIsAnimatable=True,Low=2,High=2,WeightMap="StretchAnisoDamping")
   StretchStiffnessBias=(bIsAnimatable=True,Low=100,High=100,WeightMap="StretchStiffnessBias")
   StretchStiffnessWarp=(bIsAnimatable=True,Low=100000,High=10000,WeightMap="StretchStiffnessWarp")
   StretchStiffnessWeft=(bIsAnimatable=True,Low=100000,High=10000,WeightMap="StretchStiffnessWeft")
  }
 }
 SimulationBendingConfig_1 {
  type=FChaosClothAssetSimulationBendingConfigNode
  guid=#205
  inputPins [
   Collection {type=FManagedArrayCollection}
   FlatnessRatio.WeightMap {type=FString}
   RestAngle.WeightMap {type=FString}
   BendingStiffness.WeightMap {type=FString}
   BendingStiffnessWarp.WeightMap {type=FString}
   BendingStiffnessWeft.WeightMap {type=FString}
   BendingStiffnessBias.WeightMap {type=FString}
   BendingDamping.WeightMap {type=FString}
   BendingAnisoDamping.WeightMap {type=FString}
   AnisoBucklingRatio.WeightMap {type=FString}
   BucklingStiffness.WeightMap {type=FString}
   BucklingStiffnessWarp.WeightMap {type=FString}
   BucklingStiffnessWeft.WeightMap {type=FString}
   BucklingStiffnessBias.WeightMap {type=FString}
   BucklingRatioWeighted.WeightMap {type=FString}
  ]
  properties {
   AnisoBucklingRatio=(bIsAnimatable=True,Low=0.5,High=0.5,WeightMap="BucklingRatio")
   BendingAnisoDamping=(bIsAnimatable=True,Low=10,High=1,WeightMap="BendingAnisoDamping")
   BendingStiffnessBias=(bIsAnimatable=True,Low=1000,High=100,WeightMap="BendingStiffnessBias")
   BendingStiffnessWarp=(bIsAnimatable=True,Low=1000,High=100,WeightMap="BendingStiffnessWarp")
   BendingStiffnessWeft=(bIsAnimatable=True,Low=1000,High=100,WeightMap="BendingStiffnessWeft")
   BucklingStiffnessBias=(bIsAnimatable=True,Low=50,High=50,WeightMap="BucklingStiffnessBias")
   BucklingStiffnessWarp=(bIsAnimatable=True,Low=50,High=50,WeightMap="BucklingStiffnessWarp")
   BucklingStiffnessWeft=(bIsAnimatable=True,Low=50,High=50,WeightMap="BucklingStiffnessWeft")
   SolverType=PBD
  }
 }
 SimulationStretchConfig_2 {
  type=FChaosClothAssetSimulationStretchConfigNode
  guid=#206
  properties {
   SolverType=PBD
   StretchAnisoDamping=(bIsAnimatable=True,Low=2,High=2,WeightMap="StretchAnisoDamping")
   StretchStiffnessBias=(bIsAnimatable=True,Low=100,High=100,WeightMap="StretchStiffnessBias")
   StretchStiffnessWarp=(bIsAnimatable=True,Low=100000,High=10000,WeightMap="StretchStiffnessWarp")
   StretchStiffnessWeft=(bIsAnimatable=True,Low=100000,High=10000,WeightMap="StretchStiffnessWeft")
  }
 }
 SimulationBendingConfig_2 {
  type=FChaosClothAssetSimulationBendingConfigNode
  guid=#207
  inputPins [
   Collection {type=FManagedArrayCollection}
   FlatnessRatio.WeightMap {type=FString}
   RestAngle.WeightMap {type=FString}
   BendingStiffness.WeightMap {type=FString}
   BendingStiffnessWarp.WeightMap {type=FString}
   BendingStiffnessWeft.WeightMap {type=FString}
   BendingStiffnessBias.WeightMap {type=FString}
   BendingDamping.WeightMap {type=FString}
   BendingAnisoDamping.WeightMap {type=FString}
   AnisoBucklingRatio.WeightMap {type=FString}
   BucklingStiffness.WeightMap {type=FString}
   BucklingStiffnessWarp.WeightMap {type=FString}
   BucklingStiffnessWeft.WeightMap {type=FString}
   BucklingStiffnessBias.WeightMap {type=FString}
   BucklingRatioWeighted.WeightMap {type=FString}
  ]
  properties {
   AnisoBucklingRatio=(bIsAnimatable=True,Low=0.5,High=0.5,WeightMap="BucklingRatio")
   BendingAnisoDamping=(bIsAnimatable=True,Low=3,High=1,WeightMap="BendingAnisoDamping")
   BendingStiffnessBias=(bIsAnimatable=True,Low=300,High=100,WeightMap="BendingStiffnessBias")
   BendingStiffnessWarp=(bIsAnimatable=True,Low=300,High=100,WeightMap="BendingStiffnessWarp")
   BendingStiffnessWeft=(bIsAnimatable=True,Low=300,High=100,WeightMap="BendingStiffnessWeft")
   BucklingStiffnessBias=(bIsAnimatable=True,Low=50,High=50,WeightMap="BucklingStiffnessBias")
   BucklingStiffnessWarp=(bIsAnimatable=True,Low=50,High=50,WeightMap="BucklingStiffnessWarp")
   BucklingStiffnessWeft=(bIsAnimatable=True,Low=50,High=50,WeightMap="BucklingStiffnessWeft")
   SolverType=PBD
  }
 }
 SimulationStretchConfig_3 {
  type=FChaosClothAssetSimulationStretchConfigNode
  guid=#208
  properties {
   SolverType=PBD
   StretchAnisoDamping=(bIsAnimatable=True,Low=2,High=2,WeightMap="StretchAnisoDamping")
   StretchStiffnessBias=(bIsAnimatable=True,Low=100,High=100,WeightMap="StretchStiffnessBias")
   StretchStiffnessWarp=(bIsAnimatable=True,Low=100000,High=10000,WeightMap="StretchStiffnessWarp")
   StretchStiffnessWeft=(bIsAnimatable=True,Low=100000,High=10000,WeightMap="StretchStiffnessWeft")
  }
 }
 SimulationBendingConfig_3 {
  type=FChaosClothAssetSimulationBendingConfigNode
  guid=#209
  inputPins [
   Collection {type=FManagedArrayCollection}
   FlatnessRatio.WeightMap {type=FString}
   RestAngle.WeightMap {type=FString}
   BendingStiffness.WeightMap {type=FString}
   BendingStiffnessWarp.WeightMap {type=FString}
   BendingStiffnessWeft.WeightMap {type=FString}
   BendingStiffnessBias.WeightMap {type=FString}
   BendingDamping.WeightMap {type=FString}
   BendingAnisoDamping.WeightMap {type=FString}
   AnisoBucklingRatio.WeightMap {type=FString}
   BucklingStiffness.WeightMap {type=FString}
   BucklingStiffnessWarp.WeightMap {type=FString}
   BucklingStiffnessWeft.WeightMap {type=FString}
   BucklingStiffnessBias.WeightMap {type=FString}
   BucklingRatioWeighted.WeightMap {type=FString}
  ]
  properties {
   AnisoBucklingRatio=(bIsAnimatable=True,Low=0.5,High=0.5,WeightMap="BucklingRatio")
   BendingAnisoDamping=(bIsAnimatable=True,Low=1,High=1,WeightMap="BendingAnisoDamping")
   BendingStiffnessBias=(bIsAnimatable=True,Low=100,High=100,WeightMap="BendingStiffnessBias")
   BendingStiffnessWarp=(bIsAnimatable=True,Low=100,High=100,WeightMap="BendingStiffnessWarp")
   BendingStiffnessWeft=(bIsAnimatable=True,Low=100,High=100,WeightMap="BendingStiffnessWeft")
   BucklingStiffnessBias=(bIsAnimatable=True,Low=50,High=50,WeightMap="BucklingStiffnessBias")
   BucklingStiffnessWarp=(bIsAnimatable=True,Low=50,High=50,WeightMap="BucklingStiffnessWarp")
   BucklingStiffnessWeft=(bIsAnimatable=True,Low=50,High=50,WeightMap="BucklingStiffnessWeft")
   SolverType=PBD
  }
 }
 MaterialWeight_3 {
  type=FGetDataflowVariableNode
  guid=#210
  outputPins [Value {type=int32}]
  properties {
   Category=Condition
   VariableName=MaterialWeight
   VariablePropertyBag=(Value=&PropertyBag_33cf849a4a835154(MaterialWeight=1))
  }
 }
 Select_1 {
  type=FDataflowSelectNode
  guid=#211
  inputPins [
   SelectedIndex {type=int32}
   "Inputs[0]" {type=FManagedArrayCollection}
   "Inputs[1]" {type=FManagedArrayCollection}
   "Inputs[2]" {type=FManagedArrayCollection}
  ]
  outputPins [Result {type=FManagedArrayCollection}]
  properties {SelectedIndex=0}
 }
]
